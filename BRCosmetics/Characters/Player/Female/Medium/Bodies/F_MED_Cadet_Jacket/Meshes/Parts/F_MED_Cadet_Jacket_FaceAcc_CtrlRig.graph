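# Control Rig for the face accessory part of the female medium Cadet Jacket body, adjusting the roots of the two hair braids.
# On forward solve, it aims the left and right braid master bones at offset points in spine_05 space.
# Two Modify Transforms steps then add a small translation nudge to each master, carrying the child braid chains along.
# Earring and chin strap bones exist in the hierarchy but are not driven.
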ControlRigBlueprint /BRCosmetics/Characters/Player/Female/Medium/Bodies/F_MED_Cadet_Jacket/Meshes/Parts/F_MED_Cadet_Jacket_FaceAcc_CtrlRig
graphs [
 RigVMFunctionLibrary {links [], nodes []}
 RigVMModel {nodes [
   RigUnit_BeginExecution = &RigUnit_BeginExecution()
   /Script/ControlRig.RigUnit_AimItem(
    DebugSettings=(Scale=10,WorldOffset=(Rotation=(X=0,Y=0,Z=0,W=1),Translation=(X=0,Y=0,Z=0),Scale3D=(X=1,Y=1,Z=1)))
    Item=(Type=Bone,Name="dyn_l_braid_MSTR")
    Primary=(Weight=1,Axis=(X=1,Y=0,Z=0),Target=(X=0,Y=-4,Z=-4),Kind=Location,Space=(Type=Bone,Name="spine_05"))
    Secondary=(Weight=1,Axis=(X=0,Y=0,Z=0),Target=(X=0,Y=0,Z=0),Kind=Location,Space=(Type=Bone))
    Weight=1
    ExecutePin=@RigUnit_BeginExecution.ExecutePin
   ) {is=struct RigVMUnitNode}
   > &RigUnit_AimItem(
    DebugSettings=(Scale=10,WorldOffset=(Rotation=(X=0,Y=0,Z=0,W=1),Translation=(X=0,Y=0,Z=0),Scale3D=(X=1,Y=1,Z=1)))
    Item=(Type=Bone,Name="dyn_r_braid_MSTR")
    Primary=(Weight=1,Axis=(X=1,Y=0,Z=0),Target=(X=0,Y=-3,Z=3),Kind=Location,Space=(Type=Bone,Name="spine_05"))
    Secondary=(Weight=1,Axis=(X=0,Y=0,Z=0),Target=(X=0,Y=0,Z=0),Kind=Location,Space=(Type=Bone))
    Weight=1
   )
   > &RigUnit_ModifyTransforms(
    ItemToModify=((Item=(Type=Bone),Transform=(Rotation=(X=0,Y=0,Z=0,W=1),Translation=(X=0,Y=0,Z=0),Scale3D=(X=1,Y=1,Z=1))),(Item=(Type=Bone,Name="dyn_l_braid_MSTR"),Transform=(Rotation=(X=0,Y=0,Z=0,W=1),Translation=(X=0,Y=0.2,Z=-0.2),Scale3D=(X=1,Y=1,Z=1))))
    Weight=1
   )
   > &RigUnit_ModifyTransforms(
    ItemToModify=((Item=(Type=Bone),Transform=(Rotation=(X=0,Y=0,Z=0,W=1),Translation=(X=0,Y=0,Z=0),Scale3D=(X=1,Y=1,Z=1))),(Item=(Type=Bone,Name="dyn_r_braid_MSTR"),Transform=(Rotation=(X=0,Y=0,Z=0,W=1),Translation=(X=0,Y=0.1,Z=0.05),Scale3D=(X=1,Y=1,Z=1))))
    Weight=1
   )
  ]}
]
hierarchy [
 root {loc=<0 0 0>, rot=<0 0 0 1>}
 pelvis {parent=root, loc=<0 -0.5301 98.0571>, rot=<0.0127 0.707 -0.0127 -0.707>}
 spine_01 {parent=pelvis, loc=<6.1208 -0.9328 0>, rot=<0 0 0.034 -0.9994>}
 spine_02 {parent=spine_01, loc=<6.5996 -0.2095 0>, rot=<0 0 0.0238 -0.9997>}
 spine_03 {parent=spine_02, loc=<6.4914 -0.1237 0>, rot=<0 0 0.0149 -0.9999>}
 spine_04 {parent=spine_03, loc=<6.9598 0.0794 0>, rot=<0 0 -0.0132 -0.9999>}
 spine_05 {parent=spine_04, loc=<6.9987 0 0>, rot=<0 0 -0.0015 -1>}
 clavicle_l {parent=spine_05, loc=<5.4831 -0.1053 -3.2946>, rot=<0.0465 -0.7289 -0.009 -0.6829>}
 clavicle_r {parent=spine_05, loc=<5.4831 -0.1053 3.2946>, rot=<0.7289 0.0465 -0.6829 0.009>}
 neck_01 {parent=spine_05, loc=<11.3138 0.1529 0>, rot=<0 0 -0.2003 -0.9797>}
 neck_02 {parent=neck_01, loc=<4.5627 0 0>, rot=<0 0 0.0504 -0.9987>}
 head {parent=neck_02, loc=<5.3914 0 0>, rot=<0 0 0.115 -0.9934>}
 faceAttach {parent=head, loc=<0.6201 0.9296 0>, rot=<0.0082 -0.7071 -0.0082 -0.7071>, scale=<0.89 0.89 0.89>}
 C_brow_mid {parent=faceAttach, loc=<-0.0104 9.0479 4.2428>, rot=<0 -0.0594 0.0117 -0.9982>}
 L_brow_mid {parent=faceAttach, loc=<3.4375 8.5561 5.4621>, rot=<-0.0011 -0.0122 0.2553 -0.9668>}
 L_brow_outer {parent=faceAttach, loc=<5.9151 6.3399 5.3857>, rot=<0.4823 0.6046 0.1606 -0.6133>}
 L_cheek_inner {parent=faceAttach, loc=<4.7207 7.462 -1.3828>, rot=<-0.0491 -0.2152 0.3254 -0.9194>}
 C_lip_upper_mid {parent=faceAttach, loc=<-0.0012 9.6095 -3.6272>, rot=<0.2225 0.0011 -0.0004 -0.9749>}
 L_lip_upper_outer {parent=faceAttach, loc=<1.5596 8.7151 -3.7596>, rot=<0.0242 -0.1057 0.2749 -0.9553>}
 L_lip_corner {parent=faceAttach, loc=<2.6448 8.1386 -3.7867>, rot=<0.0027 0.0083 -0.0146 -0.9999>}
 C_nose_bridge {parent=faceAttach, loc=<-0.0014 10.1588 0.7361>, rot=<-0.2218 -0.0041 -0.0011 -0.9751>}
 R_brow_mid {parent=faceAttach, loc=<-3.4735 8.5435 5.4672>, rot=<-0.0123 0.0978 -0.2548 -0.9619>}
 R_brow_outer {parent=faceAttach, loc=<-5.9136 6.3404 5.3924>, rot=<0.0805 -0.0975 -0.4201 -0.8986>}
 R_cheek_inner {parent=faceAttach, loc=<-4.7207 7.462 -1.3828>, rot=<-0.0495 0.2164 -0.3253 -0.9192>}
 R_lip_corner {parent=faceAttach, loc=<-2.6775 8.1203 -3.7575>, rot=<0.0029 -0.008 0.0146 -0.9999>}
 C_jaw {parent=faceAttach, loc=<-0.0199 0.5419 -2.5446>, rot=<0 0 0 1>}
 L_lip_lower_outer {parent=C_jaw, loc=<1.5941 8.0622 -1.2113>, rot=<0.0279 0.1256 0.3654 -0.9219>}
 C_lip_lower_mid {parent=C_jaw, loc=<0.0126 8.8736 -1.8066>, rot=<0 -0.0096 -0.002 -1>}
 R_lip_lower_outer {parent=C_jaw, loc=<-1.4462 8.3641 -1.4114>, rot=<0.0199 -0.0819 -0.3336 -0.939>}
 R_eye {parent=faceAttach, loc=<-3.4152 5.9999 3.2423>, rot=<0 0 0.9998 -0.0187>}
 R_eye_lid_upper_mid {parent=R_eye, loc=<-0.0396 -1.1714 0.2631>, rot=<0.0523 -0.0009 0.0167 -0.9985>}
 R_eye_lid_lower_mid {parent=R_eye, loc=<-0.0085 -0.9425 -0.7051>, rot=<0.0329 -0.0005 0.0167 -0.9993>}
 L_eye {parent=faceAttach, loc=<3.3268 5.9941 3.2422>, rot=<0 0 0.9998 0.0187>}
 L_eye_lid_upper_mid {parent=L_eye, loc=<0.0251 -1.1623 0.2388>, rot=<0.0523 0 0 -0.9986>}
 L_eye_lid_lower_mid {parent=L_eye, loc=<0.0223 -0.923 -0.7488>, rot=<0.0329 0 0 -0.9995>}
 R_lip_upper_outer {parent=faceAttach, loc=<-1.5599 8.7146 -3.7594>, rot=<0.0222 0.1114 -0.277 -0.9541>}
 dyn_Hair_Braid_MSTR {parent=head, loc=<0 0 0>, rot=<0 0 0 1>}
 dyn_r_braid_MSTR {parent=dyn_Hair_Braid_MSTR, loc=<9.4381 -5.2419 4.3432>, rot=<0.0065 -1 0 0>}
 dyn_r_braid_01 {parent=dyn_r_braid_MSTR, loc=<0 0 0>, rot=<0 0.0054 -0.0118 -0.9999>}
 dyn_r_braid_02 {parent=dyn_r_braid_01, loc=<10.3753 0 0>, rot=<0.0055 -0.0278 0.2118 -0.9769>}
 dyn_r_braid_03 {parent=dyn_r_braid_02, loc=<10.1307 0 0>, rot=<0.0021 -0.0044 0.0317 -0.9995>}
 dyn_r_braid_04 {parent=dyn_r_braid_03, loc=<8.6659 0 0>, rot=<-0.0062 0.0175 -0.1298 -0.9914>}
 dyn_r_braid_05 {parent=dyn_r_braid_04, loc=<10.1868 0 0>, rot=<-0.0008 0.0043 -0.0338 -0.9994>}
 dyn_r_braid_06 {parent=dyn_r_braid_05, loc=<10.0394 0 0>, rot=<-0.063 -0.0016 0.0248 -0.9977>}
 dyn_l_braid_MSTR {parent=dyn_Hair_Braid_MSTR, loc=<9.438 -5.242 -4.343>, rot=<0.0041 -1 0 0>}
 dyn_l_braid_01 {parent=dyn_l_braid_MSTR, loc=<0 0 0>, rot=<-0.0001 -0.0058 -0.0117 -0.9999>}
 dyn_l_braid_02 {parent=dyn_l_braid_01, loc=<10.3753 0 0>, rot=<-0.0054 0.0283 0.2118 -0.9769>}
 dyn_l_braid_03 {parent=dyn_l_braid_02, loc=<10.1307 0 0>, rot=<-0.0021 0.0045 0.0317 -0.9995>}
 dyn_l_braid_04 {parent=dyn_l_braid_03, loc=<8.6659 0 0>, rot=<0.0062 -0.0178 -0.1298 -0.9914>}
 dyn_l_braid_05 {parent=dyn_l_braid_04, loc=<10.1868 0 0>, rot=<0.0005 -0.0029 -0.0221 -0.9998>}
 dyn_l_braid_06 {parent=dyn_l_braid_05, loc=<9.9774 0 0>, rot=<0.0644 0.0008 0.013 -0.9978>}
 dyn_earring_R_MSTR {parent=head, loc=<0.0333 0.2881 6.2833>, rot=<0.0041 -1 0 0>}
 dyn_earring_R_01 {parent=dyn_earring_R_MSTR, loc=<0 0 0>, rot=<0 0 0 1>}
 dyn_earring_L_MSTR {parent=head, loc=<0.167 0.4199 -6.3037>, rot=<0.0041 -1 0 0>}
 dyn_earring_L_01 {parent=dyn_earring_L_MSTR, loc=<0 0 0>, rot=<0 0 0 1>}
 dyn_chinStrap {parent=head, loc=<-6.5878 4.7712 0>, rot=<0.0029 -0.7071 -0.0029 -0.7071>}
]
