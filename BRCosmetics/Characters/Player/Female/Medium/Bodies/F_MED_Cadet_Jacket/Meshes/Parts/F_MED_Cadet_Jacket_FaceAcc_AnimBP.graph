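# Post-pose AnimBP for the female Cadet Jacket face accessory part, animating braids and earrings.
# A cached pre-sim pose blends copied parent and body poses, adds a melee Control Rig and head-driven braid offsets.
# Both branches simulate earrings with AnimDynamics; braids use RigidBody, or bone constraints plus AnimDynamics when a backbling is worn or rigid bodies are disabled.
# The EventGraph updates these switches, gravity, and skydive and aiming adjustments.
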AnimBlueprint /BRCosmetics/Characters/Player/Female/Medium/Bodies/F_MED_Cadet_Jacket/Meshes/Parts/F_MED_Cadet_Jacket_FaceAcc_AnimBP
target_skeleton=&/F_MED_Cadet_Jacket_FaceAcc_Skeleton
graphs [
 AnimGraph {nodes [
   "Constraint - dyn_beltPouch_a (T)"(Alpha=1, ConstraintWeights_0=1) {
    alphaBoolBlend {blendOption=Linear}
    boneToModify {boneName=dyn_beltPouch_a}
    constraintSetup [{transformType=Translation, targetBone {boneName=None}}]
    constraintWeights [1]
   }
   n1 = Anim Dynamics - Bone: dyn_earring_L_01(ComponentPose=Local To Component(LocalPose=Use cached pose 'PreSim'()), Alpha=1, ExternalForce="0,0,0") {
    angularDampingOverride=0.993
    bGravityOverrideInSimSpace=true
    bOverrideAngularDamping=true
    bUseGravityOverride=true
    simulationSpace=RootRelative
    alphaBoolBlend {blendOption=Linear}
    boundBone {boneName=dyn_earring_L_01}
    componentAppliedLinearAccClamp=<3000 3000 3000>
    componentLinearAccScale=<0.3 0.3 0.3>
    componentLinearVelScale=<0.3 0.3 0.3>
    gravityOverride {z=-980}
    physicsBodyDefinitions [{
      sphereCollisionRadius=0
      boundBone {boneName=dyn_earring_L_01}
      boxExtents=<8 8 2.5>
      constraintSetup {angularLimitsMax=<3 5 5>, angularLimitsMin=<-3 -15 -30>, angularTarget=<0 0 0>}
      localJointOffset=<0.75 0 0>
     }]
    relativeSpaceBone {boneName=dyn_earring_L_MSTR}
   }
   n2 = Anim Dynamics - Bone: dyn_earring_R_01(ComponentPose=@n1, Alpha=1, ExternalForce="0,0,0") {
    angularDampingOverride=0.993
    bGravityOverrideInSimSpace=true
    bOverrideAngularDamping=true
    bUseGravityOverride=true
    simulationSpace=RootRelative
    alphaBoolBlend {blendOption=Linear}
    boundBone {boneName=dyn_earring_R_01}
    componentAppliedLinearAccClamp=<3000 3000 3000>
    componentLinearAccScale=<0.3 0.3 0.3>
    componentLinearVelScale=<0.3 0.3 0.3>
    gravityOverride {z=-980}
    physicsBodyDefinitions [{
      sphereCollisionRadius=0
      boundBone {boneName=dyn_earring_R_01}
      boxExtents=<10 10 2.75>
      constraintSetup {angularLimitsMax=<3 15 3>, angularLimitsMin=<-3 -5 -25>, angularTarget=<0 0 0>}
      localJointOffset=<0.75 0 0>
     }]
    relativeSpaceBone {boneName=dyn_earring_R_MSTR}
   }
   n3 = Get Skydive ROT Alpha() -> (Skydive ROT Alpha:double)
   n4 = "Transform (Modify) Bone - Bone: dyn_r_braid_MSTR"(ComponentPose=@n2, Alpha=@n3, Translation="0,-1,0", Rotation="0,-55,0", Scale="1,1,1") {
    rotationMode=BMM_Additive
    rotationSpace=BCS_BoneSpace
    translationMode=BMM_Additive
    translationSpace=BCS_BoneSpace
    alphaBoolBlend {blendOption=Linear}
    alphaScaleBiasClamp {bInterpResult=true}
    boneToModify {boneName=dyn_r_braid_MSTR}
    rotation {yaw=-55}
    translation {y=-1}
   }
   n5 = "Transform (Modify) Bone - Bone: dyn_l_braid_MSTR"(ComponentPose=@n4, Alpha=@n3, Translation="0,-1,0", Rotation="0,-55,0", Scale="1,1,1") {
    rotationMode=BMM_Additive
    rotationSpace=BCS_BoneSpace
    translationMode=BMM_Additive
    translationSpace=BCS_BoneSpace
    alphaBoolBlend {blendOption=Linear}
    alphaScaleBiasClamp {bInterpResult=true}
    boneToModify {boneName=dyn_l_braid_MSTR}
    rotation {yaw=-55}
    translation {y=-1}
   }
   n6 = RigidBody!(
    ComponentPose=@n5
    Alpha=Get Skydive RBAN Alpha() -> (Skydive RBAN Alpha:double)
    OverrideWorldGravity:Vector=Get GravityOverride()
    ExternalForce
    ComponentLinearAccScale:Vector=Get ADS Hair Velocity() -> (ADS Hair Velocity:Vector)
   ) -> (Pose) {
    bClampLinearTranslationLimitToRefPose=true
    bOverrideWorldGravity=true
    overridePhysicsAsset=&/F_MED_Cadet_Jacket_FaceAcc_Physics
    simulationSpace=BaseBoneSpace
    alphaBoolBlend {blendOption=Linear}
    alphaScaleBiasClamp {bInterpResult=true}
    baseBoneRef {boneName=neck_01}
    componentAppliedLinearAccClamp=<3000 3000 3000>
    componentLinearAccScale=<0.3 0.3 0.45>
    componentLinearVelScale=<0.35 0.35 0.5>
    overrideWorldGravity {z=-980}
   }
   n7 = Get IsBackblingOnAlpha() -> (IsBackblingOnAlpha:double)
   n8 = Anim Dynamics - Bone: dyn_earring_L_01(ComponentPose=Local To Component(LocalPose=Use cached pose 'PreSim'()), Alpha=1, ExternalForce="0,0,0") {
    angularDampingOverride=0.993
    bGravityOverrideInSimSpace=true
    bOverrideAngularDamping=true
    bUseGravityOverride=true
    simulationSpace=RootRelative
    alphaBoolBlend {blendOption=Linear}
    boundBone {boneName=dyn_earring_L_01}
    componentAppliedLinearAccClamp=<3000 3000 3000>
    componentLinearAccScale=<0.3 0.3 0.3>
    componentLinearVelScale=<0.3 0.3 0.3>
    gravityOverride {z=-980}
    physicsBodyDefinitions [{
      sphereCollisionRadius=0
      boundBone {boneName=dyn_earring_L_01}
      boxExtents=<8 8 2.5>
      constraintSetup {angularLimitsMax=<3 5 5>, angularLimitsMin=<-3 -15 -30>, angularTarget=<0 0 0>}
      localJointOffset=<0.75 0 0>
     }]
    relativeSpaceBone {boneName=dyn_earring_L_MSTR}
   }
   n9 = Anim Dynamics - Bone: dyn_earring_R_01(ComponentPose=@n8, Alpha=1, ExternalForce="0,0,0") {
    angularDampingOverride=0.993
    bGravityOverrideInSimSpace=true
    bOverrideAngularDamping=true
    bUseGravityOverride=true
    simulationSpace=RootRelative
    alphaBoolBlend {blendOption=Linear}
    boundBone {boneName=dyn_earring_R_01}
    componentAppliedLinearAccClamp=<3000 3000 3000>
    componentLinearAccScale=<0.3 0.3 0.3>
    componentLinearVelScale=<0.3 0.3 0.3>
    gravityOverride {z=-980}
    physicsBodyDefinitions [{
      sphereCollisionRadius=0
      boundBone {boneName=dyn_earring_R_01}
      boxExtents=<10 10 2.75>
      constraintSetup {angularLimitsMax=<3 15 3>, angularLimitsMin=<-3 -5 -25>, angularTarget=<0 0 0>}
      localJointOffset=<0.75 0 0>
     }]
    relativeSpaceBone {boneName=dyn_earring_R_MSTR}
   }
   n10 = "Constraint - dyn_r_braid_01 (PPP)"!(ComponentPose=@n9, Alpha=@n7, ConstraintWeights_0=0.4, ConstraintWeights_1:float=0.4, ConstraintWeights_2:float=0.2) -> (Pose) {
    alphaBoolBlend {blendOption=Linear}
    boneToModify {boneName=dyn_r_braid_01}
    constraintSetup [{targetBone {boneName=head}}, {targetBone {boneName=neck_02}}, {targetBone {boneName=clavicle_r}}]
    constraintWeights [1,1,1]
   }
   n11 = Get HeadBack_Alpha() -> (HeadBack_Alpha:double)
   n12 = "Transform (Modify) Bone - Bone: dyn_r_braid_01"(ComponentPose=@n10, Alpha=@n11, Translation="-2,-1.5,0", Rotation="0,-25,0", Scale="1,1,1") {
    rotationMode=BMM_Additive
    rotationSpace=BCS_BoneSpace
    translationMode=BMM_Additive
    translationSpace=BCS_BoneSpace
    alphaBoolBlend {blendOption=Linear}
    alphaScaleBiasClamp {bInterpResult=true}
    boneToModify {boneName=dyn_r_braid_01}
    rotation {yaw=-25}
    translation=<-2 -1.5>
   }
   n13 = "Constraint - dyn_r_braid_02 (PPP)"!(ComponentPose=@n12, Alpha=@n7, ConstraintWeights_0=0.2, ConstraintWeights_1:float=0.5, ConstraintWeights_2:float=0.3) -> (Pose) {
    alphaBoolBlend {blendOption=Linear}
    boneToModify {boneName=dyn_r_braid_02}
    constraintSetup [{targetBone {boneName=neck_02}}, {targetBone {boneName=spine_05}}, {targetBone {boneName=clavicle_r}}]
    constraintWeights [1,1,1]
   }
   n14 = "Constraint - dyn_r_braid_03 (PPP)"!(ComponentPose=@n13, Alpha=@n7, ConstraintWeights_0=0.4, ConstraintWeights_1:float=0.4, ConstraintWeights_2:float=0.2) -> (Pose) {
    alphaBoolBlend {blendOption=Linear}
    boneToModify {boneName=dyn_r_braid_03}
    constraintSetup [{targetBone {boneName=neck_01}}, {targetBone {boneName=spine_05}}, {targetBone {boneName=clavicle_r}}]
    constraintWeights [1,1,1]
   }
   n15 = "Constraint - dyn_r_braid_04 (P)"(ComponentPose=@n14, Alpha=@n7, ConstraintWeights_0=1) {
    alphaBoolBlend {blendOption=Linear}
    boneToModify {boneName=dyn_r_braid_04}
    constraintSetup [{targetBone {boneName=spine_05}}]
    constraintWeights [1]
   }
   n16 = "Constraint - dyn_r_braid_05 (P)"(ComponentPose=@n15, Alpha=@n7, ConstraintWeights_0=1) {
    alphaBoolBlend {blendOption=Linear}
    boneToModify {boneName=dyn_r_braid_05}
    constraintSetup [{targetBone {boneName=spine_03}}]
    constraintWeights [1]
   }
   n17 = "Transform (Modify) Bone - Bone: dyn_r_braid_03"(ComponentPose=@n16, Alpha=1, Translation="0,-1,0", Rotation="0,10,0", Scale="1,1,1") {
    rotationMode=BMM_Additive
    rotationSpace=BCS_BoneSpace
    translationMode=BMM_Additive
    translationSpace=BCS_BoneSpace
    alphaBoolBlend {blendOption=Linear}
    alphaScaleBiasClamp {bInterpResult=true}
    boneToModify {boneName=dyn_r_braid_03}
    rotation {yaw=10}
    translation {y=-1}
   }
   n18 = "Transform (Modify) Bone - Bone: dyn_r_braid_04"(ComponentPose=@n17, Alpha=1, Translation="0,0,0", Rotation="0,10,0", Scale="1,1,1") {
    rotationMode=BMM_Additive
    rotationSpace=BCS_BoneSpace
    translationMode=BMM_Additive
    translationSpace=BCS_BoneSpace
    alphaBoolBlend {blendOption=Linear}
    alphaScaleBiasClamp {bInterpResult=true}
    boneToModify {boneName=dyn_r_braid_04}
    rotation {yaw=10}
   }
   n19 = "Constraint - dyn_l_braid_01 (PPP)"!(ComponentPose=@n18, Alpha=@n7, ConstraintWeights_0=0.4, ConstraintWeights_1:float=0.4, ConstraintWeights_2:float=0.2) -> (Pose) {
    alphaBoolBlend {blendOption=Linear}
    boneToModify {boneName=dyn_l_braid_01}
    constraintSetup [{targetBone {boneName=head}}, {targetBone {boneName=neck_02}}, {targetBone {boneName=clavicle_l}}]
    constraintWeights [1,1,1]
   }
   n20 = "Transform (Modify) Bone - Bone: dyn_l_braid_01"(ComponentPose=@n19, Alpha=@n11, Translation="-2,-1.5,0", Rotation="0,-25,0", Scale="1,1,1") {
    rotationMode=BMM_Additive
    rotationSpace=BCS_BoneSpace
    translationMode=BMM_Additive
    translationSpace=BCS_BoneSpace
    alphaBoolBlend {blendOption=Linear}
    alphaScaleBiasClamp {bInterpResult=true}
    boneToModify {boneName=dyn_l_braid_01}
    rotation {yaw=-25}
    translation=<-2 -1.5>
   }
   n21 = "Constraint - dyn_l_braid_02 (PPP)"!(ComponentPose=@n20, Alpha=@n7, ConstraintWeights_0=0.2, ConstraintWeights_1:float=0.5, ConstraintWeights_2:float=0.3) -> (Pose) {
    alphaBoolBlend {blendOption=Linear}
    boneToModify {boneName=dyn_l_braid_02}
    constraintSetup [{targetBone {boneName=neck_02}}, {targetBone {boneName=spine_05}}, {targetBone {boneName=clavicle_l}}]
    constraintWeights [1,1,1]
   }
   n22 = "Constraint - dyn_l_braid_03 (PPP)"!(ComponentPose=@n21, Alpha=@n7, ConstraintWeights_0=0.4, ConstraintWeights_1:float=0.4, ConstraintWeights_2:float=0.2) -> (Pose) {
    alphaBoolBlend {blendOption=Linear}
    boneToModify {boneName=dyn_l_braid_03}
    constraintSetup [{targetBone {boneName=neck_01}}, {targetBone {boneName=spine_05}}, {targetBone {boneName=clavicle_l}}]
    constraintWeights [1,1,1]
   }
   n23 = "Constraint - dyn_l_braid_04 (P)"(ComponentPose=@n22, Alpha=@n7, ConstraintWeights_0=1) {
    alphaBoolBlend {blendOption=Linear}
    boneToModify {boneName=dyn_l_braid_04}
    constraintSetup [{targetBone {boneName=spine_05}}]
    constraintWeights [1]
   }
   n24 = "Constraint - dyn_l_braid_05 (P)"(ComponentPose=@n23, Alpha=@n7, ConstraintWeights_0=1) {
    alphaBoolBlend {blendOption=Linear}
    boneToModify {boneName=dyn_l_braid_05}
    constraintSetup [{targetBone {boneName=spine_03}}]
    constraintWeights [1]
   }
   n25 = "Transform (Modify) Bone - Bone: dyn_l_braid_03"(ComponentPose=@n24, Alpha=1, Translation="0,-1,0", Rotation="0,10,0", Scale="1,1,1") {
    rotationMode=BMM_Additive
    rotationSpace=BCS_BoneSpace
    translationMode=BMM_Additive
    translationSpace=BCS_BoneSpace
    alphaBoolBlend {blendOption=Linear}
    alphaScaleBiasClamp {bInterpResult=true}
    boneToModify {boneName=dyn_l_braid_03}
    rotation {yaw=10}
    translation {y=-1}
   }
   n26 = "Transform (Modify) Bone - Bone: dyn_l_braid_04"(ComponentPose=@n25, Alpha=1, Translation="0,0,0", Rotation="0,10,0", Scale="1,1,1") {
    rotationMode=BMM_Additive
    rotationSpace=BCS_BoneSpace
    translationMode=BMM_Additive
    translationSpace=BCS_BoneSpace
    alphaBoolBlend {blendOption=Linear}
    alphaScaleBiasClamp {bInterpResult=true}
    boneToModify {boneName=dyn_l_braid_04}
    rotation {yaw=10}
   }
   n27 = Anim Dynamics - Bone: dyn_r_braid_03(ComponentPose=@n26, Alpha=1, ExternalForce="0,0,0") {
    angularDampingOverride=0.9998
    angularSpringConstant=10000
    bAngularSpring=true
    bGravityOverrideInSimSpace=true
    bOverrideAngularDamping=true
    bUseGravityOverride=true
    gravityScale=0.35
    simulationSpace=RootRelative
    alphaBoolBlend {blendOption=Linear}
    alphaScaleBiasClamp {bMapRange=true, outRange {max=0.5}}
    boundBone {boneName=dyn_r_braid_03}
    componentAppliedLinearAccClamp=<3000 3000 3000>
    componentLinearAccScale=<0.35 0.35 0.45>
    componentLinearVelScale=<0.35 0.35 0.45>
    gravityOverride {z=-980}
    physicsBodyDefinitions [{
      sphereCollisionRadius=0
      boundBone {boneName=dyn_r_braid_03}
      boxExtents=<22 36 36>
      constraintSetup {angularLimitsMax=<0 5 0>, angularLimitsMin=<0 -2 -1>, angularTarget=<0 0 0>}
      localJointOffset=<7 0 0>
     }]
    sphericalLimits [
     {limitRadius=14, drivingBone {boneName=spine_03}, sphereLocalOffset=<0 0 0>}
     {limitRadius=10, drivingBone {boneName=spine_04}, sphereLocalOffset=<0 0 0>}
    ]
   }
   n28 = Anim Dynamics - Bone: dyn_r_braid_03(ComponentPose=@n27, Alpha=1, ExternalForce="0,0,0") {
    angularDampingOverride=0.9998
    angularSpringConstant=10000
    bAngularSpring=true
    bGravityOverrideInSimSpace=true
    bOverrideAngularDamping=true
    bUseGravityOverride=true
    gravityScale=0.35
    simulationSpace=RootRelative
    alphaBoolBlend {blendOption=Linear}
    alphaScaleBiasClamp {bMapRange=true, outRange {max=0.5}}
    boundBone {boneName=dyn_r_braid_03}
    componentAppliedLinearAccClamp=<3000 3000 3000>
    componentLinearAccScale=<0.35 0.35 0.45>
    componentLinearVelScale=<0.35 0.35 0.45>
    gravityOverride {z=-980}
    physicsBodyDefinitions [{
      sphereCollisionRadius=0
      boundBone {boneName=dyn_r_braid_03}
      boxExtents=<22 36 36>
      constraintSetup {angularLimitsMax=<0 15 0>, angularLimitsMin=<0 -5 -2>, angularTarget=<0 0 0>}
      localJointOffset=<7 0 0>
     }]
    sphericalLimits [
     {limitRadius=14, drivingBone {boneName=spine_03}, sphereLocalOffset=<0 0 0>}
     {limitRadius=10, drivingBone {boneName=spine_04}, sphereLocalOffset=<0 0 0>}
    ]
   }
   n29 = Anim Dynamics - Bone: dyn_r_braid_04(ComponentPose=@n28, Alpha=1, ExternalForce="0,0,0") {
    angularDampingOverride=0.9998
    angularSpringConstant=10000
    bAngularSpring=true
    bGravityOverrideInSimSpace=true
    bOverrideAngularDamping=true
    bUseGravityOverride=true
    gravityScale=0.35
    simulationSpace=RootRelative
    alphaBoolBlend {blendOption=Linear}
    alphaScaleBiasClamp {bMapRange=true, outRange {max=0.5}}
    boundBone {boneName=dyn_r_braid_04}
    componentAppliedLinearAccClamp=<3000 3000 3000>
    componentLinearAccScale=<0.35 0.35 0.45>
    componentLinearVelScale=<0.35 0.35 0.45>
    gravityOverride {z=-980}
    physicsBodyDefinitions [{
      sphereCollisionRadius=0
      boundBone {boneName=dyn_r_braid_04}
      boxExtents=<22 36 36>
      constraintSetup {angularLimitsMax=<0 25 0>, angularLimitsMin=<0 -15 -4>, angularTarget=<0 0 0>}
      localJointOffset=<7 0 0>
     }]
    sphericalLimits [
     {limitRadius=14, drivingBone {boneName=spine_03}, sphereLocalOffset=<0 0 0>}
     {limitRadius=10, drivingBone {boneName=spine_04}, sphereLocalOffset=<0 0 0>}
    ]
   }
   n30 = Anim Dynamics - Bone: dyn_r_braid_05(ComponentPose=@n29, Alpha=1, ExternalForce="0,0,0") {
    angularDampingOverride=0.9998
    angularSpringConstant=10000
    bAngularSpring=true
    bGravityOverrideInSimSpace=true
    bOverrideAngularDamping=true
    bUseGravityOverride=true
    gravityScale=0.35
    simulationSpace=RootRelative
    alphaBoolBlend {blendOption=Linear}
    alphaScaleBiasClamp {bMapRange=true, outRange {max=0.5}}
    boundBone {boneName=dyn_r_braid_05}
    componentAppliedLinearAccClamp=<3000 3000 3000>
    componentLinearAccScale=<0.35 0.35 0.45>
    componentLinearVelScale=<0.35 0.35 0.45>
    gravityOverride {z=-980}
    physicsBodyDefinitions [{
      sphereCollisionRadius=0
      boundBone {boneName=dyn_r_braid_05}
      boxExtents=<22 36 36>
      constraintSetup {angularLimitsMax=<0 45 0>, angularLimitsMin=<0 -35 -6>, angularTarget=<0 0 0>}
      localJointOffset=<7 0 0>
     }]
    sphericalLimits [
     {limitRadius=14, drivingBone {boneName=spine_03}, sphereLocalOffset=<0 0 0>}
     {limitRadius=10, drivingBone {boneName=spine_04}, sphereLocalOffset=<0 0 0>}
    ]
   }
   n31 = Anim Dynamics - Bone: dyn_l_braid_03(ComponentPose=@n30, Alpha=1, ExternalForce="0,0,0") {
    angularDampingOverride=0.9999
    angularSpringConstant=10000
    bAngularSpring=true
    bGravityOverrideInSimSpace=true
    bOverrideAngularDamping=true
    bUseGravityOverride=true
    gravityScale=0.35
    simulationSpace=RootRelative
    alphaBoolBlend {blendOption=Linear}
    alphaScaleBiasClamp {bMapRange=true, outRange {max=0.5}}
    boundBone {boneName=dyn_l_braid_03}
    componentAppliedLinearAccClamp=<3000 3000 3000>
    componentLinearAccScale=<0.3 0.3 0.4>
    componentLinearVelScale=<0.3 0.3 0.4>
    gravityOverride {z=-980}
    physicsBodyDefinitions [{
      sphereCollisionRadius=0
      boundBone {boneName=dyn_l_braid_03}
      boxExtents=<18 44 44>
      constraintSetup {angularLimitsMax=<0 2 0>, angularLimitsMin=<0 -5 -1>, angularTarget=<0 0 0>}
      localJointOffset=<9 0 0>
     }]
    sphericalLimits [
     {limitRadius=14, drivingBone {boneName=spine_03}, sphereLocalOffset=<0 0 0>}
     {limitRadius=10, drivingBone {boneName=spine_04}, sphereLocalOffset=<0 0 0>}
    ]
   }
   n32 = Anim Dynamics - Bone: dyn_l_braid_03(ComponentPose=@n31, Alpha=1, ExternalForce="0,0,0") {
    angularDampingOverride=0.9999
    angularSpringConstant=10000
    bAngularSpring=true
    bGravityOverrideInSimSpace=true
    bOverrideAngularDamping=true
    bUseGravityOverride=true
    gravityScale=0.35
    simulationSpace=RootRelative
    alphaBoolBlend {blendOption=Linear}
    alphaScaleBiasClamp {bMapRange=true, outRange {max=0.5}}
    boundBone {boneName=dyn_l_braid_03}
    componentAppliedLinearAccClamp=<3000 3000 3000>
    componentLinearAccScale=<0.3 0.3 0.4>
    componentLinearVelScale=<0.3 0.3 0.4>
    gravityOverride {z=-980}
    physicsBodyDefinitions [{
      sphereCollisionRadius=0
      boundBone {boneName=dyn_l_braid_03}
      boxExtents=<18 44 44>
      constraintSetup {angularLimitsMax=<0 5 0>, angularLimitsMin=<0 -15 -2>, angularTarget=<0 0 0>}
      localJointOffset=<9 0 0>
     }]
    sphericalLimits [
     {limitRadius=14, drivingBone {boneName=spine_03}, sphereLocalOffset=<0 0 0>}
     {limitRadius=10, drivingBone {boneName=spine_04}, sphereLocalOffset=<0 0 0>}
    ]
   }
   n33 = Anim Dynamics - Bone: dyn_l_braid_04(ComponentPose=@n32, Alpha=1, ExternalForce="0,0,0") {
    angularDampingOverride=0.9999
    angularSpringConstant=10000
    bAngularSpring=true
    bGravityOverrideInSimSpace=true
    bOverrideAngularDamping=true
    bUseGravityOverride=true
    gravityScale=0.35
    simulationSpace=RootRelative
    alphaBoolBlend {blendOption=Linear}
    alphaScaleBiasClamp {bMapRange=true, outRange {max=0.5}}
    boundBone {boneName=dyn_l_braid_04}
    componentAppliedLinearAccClamp=<3000 3000 3000>
    componentLinearAccScale=<0.3 0.3 0.4>
    componentLinearVelScale=<0.3 0.3 0.4>
    gravityOverride {z=-980}
    physicsBodyDefinitions [{
      sphereCollisionRadius=0
      boundBone {boneName=dyn_l_braid_04}
      boxExtents=<18 44 44>
      constraintSetup {angularLimitsMax=<0 15 0>, angularLimitsMin=<0 -25 -4>, angularTarget=<0 0 0>}
      localJointOffset=<9 0 0>
     }]
    sphericalLimits [
     {limitRadius=14, drivingBone {boneName=spine_03}, sphereLocalOffset=<0 0 0>}
     {limitRadius=10, drivingBone {boneName=spine_04}, sphereLocalOffset=<0 0 0>}
    ]
   }
   n34 = Anim Dynamics - Bone: dyn_l_braid_05(ComponentPose=@n33, Alpha=1, ExternalForce="0,0,0") {
    angularDampingOverride=0.9999
    angularSpringConstant=10000
    bAngularSpring=true
    bGravityOverrideInSimSpace=true
    bOverrideAngularDamping=true
    bUseGravityOverride=true
    gravityScale=0.35
    simulationSpace=RootRelative
    alphaBoolBlend {blendOption=Linear}
    alphaScaleBiasClamp {bMapRange=true, outRange {max=0.5}}
    boundBone {boneName=dyn_l_braid_05}
    componentAppliedLinearAccClamp=<3000 3000 3000>
    componentLinearAccScale=<0.3 0.3 0.4>
    componentLinearVelScale=<0.3 0.3 0.4>
    gravityOverride {z=-980}
    physicsBodyDefinitions [{
      sphereCollisionRadius=0
      boundBone {boneName=dyn_l_braid_05}
      boxExtents=<18 44 44>
      constraintSetup {angularLimitsMax=<0 35 0>, angularLimitsMin=<0 -45 -6>, angularTarget=<0 0 0>}
      localJointOffset=<9 0 0>
     }]
    sphericalLimits [
     {limitRadius=14, drivingBone {boneName=spine_03}, sphereLocalOffset=<0 0 0>}
     {limitRadius=10, drivingBone {boneName=spine_04}, sphereLocalOffset=<0 0 0>}
    ]
   }
   n35 = "Transform (Modify) Bone - Bone: dyn_r_braid_03"(ComponentPose=@n34, Alpha=1, Translation="0,0,0", Rotation="0,-5,0", Scale="0.9,1,1") {
    rotationMode=BMM_Additive
    rotationSpace=BCS_BoneSpace
    scaleMode=BMM_Additive
    scaleSpace=BCS_BoneSpace
    translationMode=BMM_Additive
    translationSpace=BCS_BoneSpace
    alphaBoolBlend {blendOption=Linear}
    boneToModify {boneName=dyn_r_braid_03}
    rotation {yaw=-5}
    scale {x=0.9}
   }
   n36 = "Transform (Modify) Bone - Bone: dyn_l_braid_03"(ComponentPose=@n35, Alpha=1, Translation="0,0,0", Rotation="0,-5,0", Scale="0.9,1,1") {
    rotationMode=BMM_Additive
    rotationSpace=BCS_BoneSpace
    scaleMode=BMM_Additive
    scaleSpace=BCS_BoneSpace
    translationMode=BMM_Additive
    translationSpace=BCS_BoneSpace
    alphaBoolBlend {blendOption=Linear}
    boneToModify {boneName=dyn_l_braid_03}
    rotation {yaw=-5}
    scale {x=0.9}
   }
   n37 = "Transform (Modify) Bone - Bone: dyn_r_braid_03"!(ComponentPose=@n36, bAlphaBoolEnabled:bool=True, Translation="0,0,0", Rotation="0,5,0", Scale="0.8,1,1") -> (Pose) {
    alphaInputType=Bool
    rotationMode=BMM_Additive
    rotationSpace=BCS_BoneSpace
    scaleMode=BMM_Additive
    scaleSpace=BCS_BoneSpace
    translationMode=BMM_Additive
    translationSpace=BCS_BoneSpace
    alphaBoolBlend {blendOption=Linear}
    boneToModify {boneName=dyn_r_braid_03}
    rotation {yaw=5}
    scale {x=0.8}
   }
   n38 = "Transform (Modify) Bone - Bone: dyn_l_braid_03"!(ComponentPose=@n37, bAlphaBoolEnabled:bool=True, Translation="0,0,0", Rotation="0,5,0", Scale="0.8,1,1") -> (Pose) {
    alphaInputType=Bool
    rotationMode=BMM_Additive
    rotationSpace=BCS_BoneSpace
    scaleMode=BMM_Additive
    scaleSpace=BCS_BoneSpace
    translationMode=BMM_Additive
    translationSpace=BCS_BoneSpace
    alphaBoolBlend {blendOption=Linear}
    boneToModify {boneName=dyn_l_braid_03}
    rotation {yaw=5}
    scale {x=0.8}
   }
   n39 = "Transform (Modify) Bone - Bone: dyn_r_braid_04"!(ComponentPose=@n38, bAlphaBoolEnabled:bool=True, Translation="0,0,0", Rotation="0,8,0", Scale="0.9,1,1") -> (Pose) {
    alphaInputType=Bool
    rotationMode=BMM_Additive
    rotationSpace=BCS_BoneSpace
    scaleMode=BMM_Additive
    scaleSpace=BCS_BoneSpace
    translationMode=BMM_Additive
    translationSpace=BCS_BoneSpace
    alphaBoolBlend {blendOption=Linear}
    boneToModify {boneName=dyn_r_braid_04}
    rotation {yaw=8}
    scale {x=0.9}
   }
   n40 = "Transform (Modify) Bone - Bone: dyn_l_braid_04"!(ComponentPose=@n39, bAlphaBoolEnabled:bool=True, Translation="0,0,0", Rotation="0,8,0", Scale="0.9,1,1") -> (Pose) {
    alphaInputType=Bool
    rotationMode=BMM_Additive
    rotationSpace=BCS_BoneSpace
    scaleMode=BMM_Additive
    scaleSpace=BCS_BoneSpace
    translationMode=BMM_Additive
    translationSpace=BCS_BoneSpace
    alphaBoolBlend {blendOption=Linear}
    boneToModify {boneName=dyn_l_braid_04}
    rotation {yaw=8}
    scale {x=0.9}
   }
   Output Pose(Result=Blend Poses by bool(
     BlendPose_0=Component To Local(ComponentPose=@n40)
     BlendPose_1=Component To Local(ComponentPose=@n6)
     bActiveValue=Get isHairConstrain() -> (isHairConstrain:bool)
    ) {blendType=Linear}) {layerGroup=None}
   n41 = Layered blend per bone(
    BasePose=Copy Pose From Mesh() {bCopyCurves=true, bUseAttachedParent=true}
    BlendPoses_0=Copy Pose From Mesh(SourceMeshComponent=Get Body Mesh() -> (Body Mesh:Skeletal Mesh Component Object Reference)) {bCopyCurves=true, bUseAttachedParent=true}
   ) {layerSetup [{branchFilters [
       {boneName=upperarm_r}
       {boneName=upperarm_l}
       {boneName=clavicle_r}
       {boneName=clavicle_l}
       {boneName=dyn_beltPouch_a}
       {boneName=dyn_beltPouch_b}
      ]}]}
   n42 = Control Rig!(Source=@n41, bAlphaBoolEnabled:bool=Get bIsPlayingMeleeAnim()) -> (Pose) {
    alphaInputType=Bool
    alphaBoolBlend {blendInTime=0.1, blendOption=Linear, blendOutTime=0.5}
    controlRigAssetReference {blueprintRigClass=&/F_MED_Cadet_Jacket_FaceAcc_CtrlRig.F_MED_Cadet_Jacket_FaceAcc_CtrlRig_C}
   }
   n43 = Get HeadBack_Alpha()
   n44 = "Transform (Modify) Bone - Bone: dyn_r_braid_MSTR"(ComponentPose=Local To Component(LocalPose=@n42), Alpha=@n43, Translation="-2,-1.5,0", Rotation="0,-15,0", Scale="1,1,1") {
    rotationMode=BMM_Additive
    rotationSpace=BCS_BoneSpace
    translationMode=BMM_Additive
    translationSpace=BCS_BoneSpace
    alphaBoolBlend {blendOption=Linear}
    alphaScaleBiasClamp {bInterpResult=true}
    boneToModify {boneName=dyn_r_braid_MSTR}
    rotation {yaw=-15}
    translation=<-2 -1.5>
   }
   n45 = "Transform (Modify) Bone - Bone: dyn_l_braid_MSTR"(ComponentPose=@n44, Alpha=@n43, Translation="-2,-1.5,0", Rotation="0,-15,0", Scale="1,1,1") {
    rotationMode=BMM_Additive
    rotationSpace=BCS_BoneSpace
    translationMode=BMM_Additive
    translationSpace=BCS_BoneSpace
    alphaBoolBlend {blendOption=Linear}
    alphaScaleBiasClamp {bInterpResult=true}
    boneToModify {boneName=dyn_l_braid_MSTR}
    rotation {yaw=-15}
    translation=<-2 -1.5>
   }
   n46 = Get HeadTurn_Alpha() -> (HeadTurn_Alpha:double)
   n47 = "Transform (Modify) Bone - Bone: dyn_r_braid_MSTR"(ComponentPose=@n45, Alpha=@n46, Translation="0,0,0", Rotation="0,-5,0", Scale="1,1,1") {
    rotationMode=BMM_Additive
    rotationSpace=BCS_BoneSpace
    translationMode=BMM_Additive
    translationSpace=BCS_BoneSpace
    alphaBoolBlend {blendOption=Linear}
    alphaScaleBiasClamp {bInterpResult=true}
    boneToModify {boneName=dyn_r_braid_MSTR}
    rotation {yaw=-5}
   }
   n48 = "Transform (Modify) Bone - Bone: dyn_l_braid_MSTR"(ComponentPose=@n47, Alpha=@n46, Translation="0,0,0", Rotation="0,-3,0", Scale="1,1,1") {
    rotationMode=BMM_Additive
    rotationSpace=BCS_BoneSpace
    translationMode=BMM_Additive
    translationSpace=BCS_BoneSpace
    alphaBoolBlend {blendOption=Linear}
    alphaScaleBiasClamp {bInterpResult=true}
    boneToModify {boneName=dyn_l_braid_MSTR}
    rotation {yaw=-3}
   }
   Save cached pose 'PreSim'(Pose=Component To Local(ComponentPose=@n48)) {cacheName=PreSim}
   Get IsShortBackblingOnAlpha() -> (IsShortBackblingOnAlpha:double)
  ]}
 EventGraph {nodes [
   Event BlueprintUpdateAnimation()
   > n1 = Sequence!(execute) -> (then_0, then_1, then_2:Exec, then_3:Exec, then_4:Exec, then_5:Exec)
   > n2 = Is Valid(InputObject=Get Body Mesh())
   n3 = CheckForBackbling(LongBackblingThreshold=40)
   n4 = Branch(execute=[@n2.Is Valid, @n38.Completed], Condition=@n3)
   > Set IsBackblingOn(IsBackblingOn:bool=true) -> (Output_Get:bool)
   > n5 = Branch(Condition=@n3.bIsBackblingACape)
   > Set Is Backbling A Cape(Is Backbling A Cape:bool=true) -> (Output_Get:bool)
   > Set Cape Alpha(Cape Alpha:double=0) -> (Output_Get:double)
   Set Is Backbling A Cape(execute=@n5.else, Is Backbling A Cape=false)
   > Set Cape Alpha(Cape Alpha=1)
   Set IsBackblingOn(execute=@n4.else, IsBackblingOn=false)
   n6 = Gravity Override State Function DEPRECATED(
    execute=@n1.then_1
    CharacterPartAnimInstance=Self-Reference()
    GameTimeInSeconds=GetGameTimeInSeconds()
    GravityOverrideParams=Get Gravity Override Params()
   )
   > Set GravityOverride(GravityOverride=@n6)
   n7 = Branch(execute=@n1.then_2, Condition=Get bIsSkyDiving())
   > Set Skydive RBAN Alpha(Skydive RBAN Alpha:double=0.5) -> (Output_Get:double)
   > n8 = Branch(Condition=Get bIsParachuteOpened())
   > Set Skydive ROT Alpha(Skydive ROT Alpha:double=0.4) -> (Output_Get:double)
   Set Skydive ROT Alpha(execute=@n8.else, Skydive ROT Alpha=1)
   Set Skydive RBAN Alpha(execute=@n7.else, Skydive RBAN Alpha=1)
   > Set Skydive ROT Alpha(Skydive ROT Alpha=0)
   n9 = Get bIsTargeting()
   n10 = Get bIsPlayingMeleeAnim()
   n11 = Get bIsSkyDiving()
   Set ADS Hair Velocity(
    execute=@n1.then_3
    ADS Hair Velocity:Vector=SelectVector(A="0.15,0.15,0.25", B="0.3,0.3,0.45", bPickA=OR Boolean!(A=@n9, B=@n10, C:bool=@n11) -> (ReturnValue))
   ) -> (Output_Get:Vector)
   n12 = GetOwnerSkeletalMeshComponent()
   n13 = BreakTransform(InTransform=GetDeltaTransformFromRefPose(self=@n12, BoneName=head, BaseName=neck_02))
   Set HeadBack_Alpha(execute=@n1.then_4, HeadBack_Alpha:double=MapRangeClamped(Value=BreakRotator(InRot=@n13.Rotation).Yaw, InRangeA=5, InRangeB=-30)) -> (Output_Get:double)
   n14 = BreakRotator(InRot=BreakTransform(InTransform=GetDeltaTransformFromRefPose(self=@n12, BoneName=head, BaseName=neck_02)).Rotation)
   > Set HeadTurn_Alpha(HeadTurn_Alpha:double=MapRangeClamped(Value="Absolute (Float)"(A=@n14.Pitch), InRangeA=0, InRangeB=15)) -> (Output_Get:double)
   n15 = Get bRigidBodyEnabled()
   n16 = Branch(execute=@n1.then_5, Condition=OR Boolean!(A=Get IsBackblingOn() -> (IsBackblingOn:bool), B=NOT Boolean(A=@n15)) -> (ReturnValue))
   > n17 = Set isHairConstrain(isHairConstrain:bool=true) -> (Output_Get:bool)
   > Set IsBackblingOnAlpha(IsBackblingOnAlpha:double="To Float (Boolean)"(InBool=@n17)) -> (Output_Get:double)
   n18 = Set isHairConstrain(execute=@n16.else, isHairConstrain=false)
   > Set IsBackblingOnAlpha(IsBackblingOnAlpha="To Float (Boolean)"(InBool=@n18))
   Event BlueprintInitializeAnimation()
   > n19 = Cast To FortPlayerPawnAthena(Object=TryGetPawnOwner())
   > Set Body Mesh(Body Mesh:Skeletal Mesh Component Object Reference=GetSkeletalMeshForPartType(self=@n19)) -> (Output_Get:Skeletal Mesh Component Object Reference)
   n20 = Sequence()
   n21 = Get bRigidBodyEnabled()
   n22 = Branch(execute=@n20.then_5, Condition=OR Boolean(A=Get IsShortBackblingOn() -> (IsShortBackblingOn:bool), B=NOT Boolean(A=@n21)))
   > Set isHairConstrain(isHairConstrain=true)
   Set isHairConstrain(execute=@n22.else, isHairConstrain=false)
   n23 = Get bIsTargeting()
   n24 = Get bIsPlayingMeleeAnim()
   n25 = Get bIsSkyDiving()
   Set ADS Hair Velocity(
    execute=@n20.then_3
    ADS Hair Velocity=SelectVector(A="0.15,0.15,0.25", B="0.3,0.3,0.45", bPickA=OR Boolean!(A=@n23, B=@n24, C:bool=@n25) -> (ReturnValue))
   )
   n26 = BreakTransform(InTransform=GetDeltaTransformFromRefPose(self=Get Body Mesh(), BoneName=head, BaseName=neck_02))
   Set HeadBack_Alpha(execute=@n20.then_4, HeadBack_Alpha=MapRangeClamped(Value=BreakRotator(InRot=@n26.Rotation).Yaw, InRangeA=5, InRangeB=-30))
   n27 = Get Body Mesh()
   n28 = BreakRotator(InRot=BreakTransform(InTransform=GetDeltaTransformFromRefPose(self=@n27, BoneName=head, BaseName=neck_02)).Rotation)
   > Set HeadTurn_Alpha(HeadTurn_Alpha=MapRangeClamped(Value="Absolute (Float)"(A=@n28.Pitch), InRangeA=0, InRangeB=15))
   Is Valid(exec=@n20.then_0, InputObject=Get Body Mesh())
   n29 = CheckForBackbling(LongBackblingThreshold=40)
   > n30 = Branch(Condition=@n29)
   > n31 = Branch(Condition=OR Boolean!(A=@n29.bIsBackblingACape, B=@n29.bIsBackblingLong) -> (ReturnValue))
   > Set IsLongBackblingOn(IsLongBackblingOn:bool=true) -> (Output_Get:bool)
   > Set IsLongBackblingOnAlpha(IsLongBackblingOnAlpha:double=1) -> (Output_Get:double)
   > n32 = Set IsShortBackblingOn(IsShortBackblingOn:bool=false) -> (Output_Get:bool)
   > Set IsShortBackblingOnAlpha(IsShortBackblingOnAlpha:double="To Float (Boolean)"(InBool=@n32)) -> (Output_Get:double)
   Set IsLongBackblingOn(execute=@n31.else, IsLongBackblingOn=false)
   > Set IsLongBackblingOnAlpha(IsLongBackblingOnAlpha=0)
   > n33 = Set IsShortBackblingOn(IsShortBackblingOn=true)
   > Set IsShortBackblingOnAlpha(IsShortBackblingOnAlpha="To Float (Boolean)"(InBool=@n33))
   n34 = Set IsBackblingOn(execute=@n30.else, IsBackblingOn=false)
   > Set IsBackblingOnAlpha(IsBackblingOnAlpha="To Float (Boolean)"(InBool=@n34))
   n35 = Gravity Override State Function DEPRECATED(
    execute=@n20.then_1
    CharacterPartAnimInstance=Self-Reference()
    GameTimeInSeconds=GetGameTimeInSeconds()
    GravityOverrideParams=Get Gravity Override Params()
   )
   > Set GravityOverride(GravityOverride=@n35)
   n36 = Branch(execute=@n20.then_2, Condition=Get bIsSkyDiving())
   > Set Skydive RBAN Alpha(Skydive RBAN Alpha=0.5)
   > n37 = Branch(Condition=Get bIsParachuteOpened())
   > Set Skydive ROT Alpha(Skydive ROT Alpha=0.4)
   Set Skydive ROT Alpha(execute=@n37.else, Skydive ROT Alpha=1)
   Set Skydive RBAN Alpha(execute=@n36.else, Skydive RBAN Alpha=1)
   > Set Skydive ROT Alpha(Skydive ROT Alpha=0)
   n38 = Do Once()
   OR Boolean(A=false, B=false)
   Do Once()
   Get IsShortBackblingOn()
   Set Body Mesh(Body Mesh=GetOwnerSkeletalMeshComponent())
  ]}
]
variables [
 ADS Hair Velocity {type=Vector}
 Body Mesh {type=SkeletalMeshComponent}
 Cape Alpha {type=number}
 Gravity Override Params {type=GravityOverrideParamsStruct}
 GravityOverride {type=Vector}
 HeadBack_Alpha {type=number}
 HeadTurn_Alpha {type=number}
 Is Backbling A Cape {type=boolean}
 IsBackblingOn {type=boolean}
 IsBackblingOnAlpha {type=number}
 isHairConstrain {type=boolean}
 IsLongBackblingOn {type=boolean}
 IsLongBackblingOnAlpha {type=number}
 IsShortBackblingOn {type=boolean}
 IsShortBackblingOnAlpha {type=number}
 Skydive RBAN Alpha {type=number}
 Skydive ROT Alpha {type=number}
]
